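# Lit material for plastic debris effects, enabled for use on Niagara mesh particles.
# Particle color RGB feeds the emissive input of the MF_FX_EmissiveAdjustments function.
# Scalar parameters named Metallic, Roughness and Specular expose surface response, with no textures in the graph.
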Material /Game/Effects/Fort_Effects/Materials/Plastic/M_Particle_PlasticDebris
graph {
 properties {bAutomaticallySetUsageInEditor=false, bUsedWithNiagaraMeshParticles=true, usedShadingModels=Default Lit}
 nodes [
  MaterialFunctionCall(Emissive=ParticleColor().RGB) {materialFunction=&MF_FX_EmissiveAdjustments}
  ScalarParameter() {parameterName=Metallic}
  ScalarParameter() {defaultValue=0.25, parameterName=Roughness}
  ScalarParameter() {defaultValue=0.25, parameterName=Specular}
 ]
}
